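# Material function that animates scales on a character, apparently reptilian, by rotating material normals in tangent space, outputting the rotated result and a gradient mask.
# Time-driven sine waves along pre-skinned Z form two sweeps, broken up by mask inputs and panning noise.
# These drive the rotation angle about an axis crossing tangent-space world-down with the vertex normal.
# Static switches toggle the effect, noise, double sweep and static bump.
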MaterialFunction /Game/Global/Materials/MaterialGym/Materials/RotateNormalInTangentSpace/MF_RotateN_TangentSpace
graph {nodes [
  FunctionOutput(None=NamedRerouteUsage()) {outputName=Rotated Normal}
  n1 = Time()
  n2 = ScalarParameter() {defaultValue=0.09, parameterName=ScalesSweepSpeed}
  n3 = PreSkinnedPosition()
  n4 = ComponentMask(None=ShaderStageSwitch(PixelShader=VertexInterpolator(VS=@n3).PS, VertexShader=@n3)) {b=true}
  n5 = ScalarParameter() {defaultValue=300, parameterName=ScalesSweepSize}
  n6 = Sine(None=Add(A=Multiply(A=@n1, B=@n2) {constB=0.15}, B=Divide(A=@n4, B=@n5) {constB=-125}))
  n7 = Saturate(None=StaticSwitch(True=Abs(None=@n6), False=@n6, Value=StaticBoolParameter() {parameterName=ScalesEnableDoubleSweep}))
  n8 = ScalarParameter() {defaultValue=2, parameterName=ScalesSweeoSmoothness}
  FunctionOutput(None=NamedRerouteDeclaration(None=Power(Base=@n7, Exp=@n8) {constExponent=8}) {name=GradientMask_01, nodeColor=<rgba 0 1 0.3906 1>}) {outputName=GradientMask, sortPriority=1}
  n9 = FunctionInput() {bUsePreviewValueAsDefault=true, inputName=Tech Art Mask, sortPriority=2, previewValue=<1 1 1 1>}
  NamedRerouteDeclaration(None=ChannelMaskParameter(None=@n9) {parameterName=ScalesMask}.R) {name=ScalesMask, nodeColor=<rgba 1 0 0.4219 1>}
  n10 = GetMaterialAttributes(None=NamedRerouteUsage()) {attributeGetTypes [#1, #2, #3]}
  NamedRerouteDeclaration(None=@n10.BaseColor) {name=InBaseColor, nodeColor=<rgba 0 0.7188 1 1>}
  NamedRerouteDeclaration(None=ChannelMaskParameter(None=FunctionInput() {bUsePreviewValueAsDefault=true, inputName=M_Map, sortPriority=1, previewValue=<1 1 1 1>}) {maskChannel=Green, parameterName=RandomScalesMask, defaultValue=<rgba 0 1 0 0>}.R) {name=RandomScalesMask, nodeColor=<rgba 0 1 0.3438 1>}
  NamedRerouteDeclaration(None=@n10.EmissiveColor) {name=InEmissive, nodeColor=<rgba 0 1 0.2031 1>}
  n11 = Multiply(A=@n5) {constB=0.75}
  n12 = Add(A=Time()) {constB=5}
  n13 = Multiply(A=ScalarParameter() {defaultValue=0.09, parameterName=ScalesSweepSpeed}) {constB=0.75}
  n14 = Sine(None=Add(A=Divide(A=@n4, B=@n11), B=Multiply(A=@n12, B=@n13) {constB=0.15}))
  NamedRerouteDeclaration(None=Saturate(None=Abs(None=@n14))) {name=GradientMask_02, nodeColor=<rgba 0 1 0.3906 1>}
  n15 = TextureCoordinate()
  n16 = ComponentMask(None=VectorParameter() {group=ScaleRotateNormal, parameterName=ScalesExtraSweepScale, defaultValue=<rgba 1 1 0 1>}.RGB) {g=true, r=true}
  n17 = TextureSample(UVs=Panner(Coordinate=Multiply(A=@n15, B=@n16) {constB=4}) {speedY=-0.015}) {
   samplerType=SAMPLERTYPE_LinearColor
   texture=/Game/Characters/Player/Female/Medium/Bodies/F_MED_ReptilianOcean_Sleek/Textures/T_ReptilianOcean_Sleek_Pattern_Noise
  }
  n18 = Subtract(A=Saturate(None=SmoothStep(Value=@n17.G) {constMin=0.1}), B=LinearInterpolate(Alpha=NamedRerouteUsage()) {constB=0.5})
  n19 = Multiply(
   A=LinearInterpolate(Alpha=Saturate(None=@n18)) {constA=-0.2}
   B=ScalarParameter() {defaultValue=0.07, parameterName=ScalesSecondaryNormalInfluence}
  ) {constB=0.5}
  n20 = NamedRerouteUsage()
  n21 = Multiply(
   A=TextureCoordinate()
   B=ComponentMask(None=VectorParameter() {parameterName=ScaleNoiseTiling, defaultValue=<rgba 0.75 1 0 1>}.RGB) {g=true, r=true}
  ) {constB=2}
  n22 = VectorParameter() {parameterName=ScaleNoiseSpeed, defaultValue=<rgba 0 0.075 0 1>}
  n23 = TextureSample(UVs=Panner(Coordinate=Multiply(A=@n21) {constB=2.5}, Speed=ComponentMask(None=@n22.RGB) {g=true, r=true}) {speedY=0.075}) {
   samplerType=SAMPLERTYPE_LinearColor
   texture=&T_ReptilianOcean_Sleek_Pattern_Noise
  }
  n24 = LinearInterpolate(Alpha=NamedRerouteDeclaration(None=@n23.R) {name=PanningUpNoise, nodeColor=<rgba 0.8516 1 0 1>})
  n25 = Multiply(
   A=StaticSwitchParameter(True=Multiply(A=@n20, B=@n24), False=@n20) {defaultValue=true, parameterName=EnableScaleNoise}
   B=ScalarParameter() {defaultValue=0.12, parameterName=ScalesSweepNormalRotateAmount}
  ) {constB=0.15}
  n26 = NamedRerouteUsage()
  NamedRerouteDeclaration(None=Multiply(A=Add(A=@n19, B=@n25), B=SmoothStep(Value=@n26) {constMin=0.1})) {name=NormalsRotationAmount, nodeColor=<rgba 0 1 0.5547 1>}
  NamedRerouteDeclaration(None=FunctionInput() {bUsePreviewValueAsDefault=true, inputName=Material, inputType=FunctionInput_MaterialAttributes}) {name=inMaterialAttributes, nodeColor=<rgba 1 0.0234 0 1>}
  n27 = SetMaterialAttributes!(MaterialAttributes=@n10.MaterialAttributes, BaseColor, Normal=NamedRerouteUsage(), WorldPositionOffset, EmissiveColor) {attributeSetTypes [
    #1
    #2
    #4
    #3
   ]}
  NamedRerouteDeclaration(None=@n27) {name=outMaterialAttributes, nodeColor=<rgba 0.5938 0 1 1>}
  n28 = NamedRerouteUsage()
  n29 = LinearInterpolate(B=ScalarParameter() {defaultValue=0.5, parameterName=ScalesSecondSweepRandomBlend}, Alpha=NamedRerouteUsage()) {constB=0.5}
  n30 = NamedRerouteUsage()
  n31 = Multiply(A=Saturate(None=Subtract(A=@n28, B=@n29)), B=@n30)
  n32 = Multiply(A=@n30) {constB=0.5}
  n33 = Multiply(A=Saturate(None=Add(A=@n31, B=@n32))) {constB=0.9}
  n34 = ScalarParameter() {defaultValue=1, parameterName=ScalesSweepRandomBlend}
  n35 = NamedRerouteUsage()
  n36 = Saturate(None=Subtract(A=NamedRerouteUsage(), B=LinearInterpolate(B=@n34, Alpha=@n35)))
  n37 = NamedRerouteUsage()
  n38 = Saturate(None=Add(A=Multiply(A=@n36, B=@n37), B=Multiply(A=@n37) {constB=0.75}))
  NamedRerouteDeclaration(None=StaticSwitchParameter(True=Constant() {r=1}, False=LinearInterpolate(A=@n33, B=@n38, Alpha=@n38)) {group=ScaleRotateNormal, parameterName=StaticSweepBump}) {name=ScaleSweepBump, nodeColor=<rgba 0 0.7422 1 1>}
  NamedRerouteDeclaration(None=@n10.Normal) {name=InNormal, nodeColor=<rgba 1 0 0.3984 1>}
  n39 = NamedRerouteUsage()
  n40 = Transform(None=Constant3Vector() {constant=<rgba 0 0 -1 0>}) {transformSourceType=TRANSFORMSOURCE_World, transformType=TRANSFORM_Tangent}
  n41 = VertexNormalWS()
  n42 = RotateAboutAxis(
   NormalizedRotationAxis=CrossProduct(A=Normalize(VectorInput=@n40), B=Transform(None=@n41) {transformSourceType=TRANSFORMSOURCE_World, transformType=TRANSFORM_Tangent})
   RotationAngle=NamedRerouteUsage()
   PivotPoint=Constant()
   Position=@n39
  )
  NamedRerouteDeclaration(None=StaticSwitchParameter(True=Add(A=@n39, B=@n42), False=@n39) {defaultValue=true, parameterName=EnableScalesEffect}) {name=OutNormal, nodeColor=<rgba 1 0 0.1406 1>}
 ]}
