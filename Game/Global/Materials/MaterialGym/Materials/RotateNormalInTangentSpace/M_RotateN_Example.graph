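# Material Gym example that demonstrates MF_RotateN_TangentSpace on a tiled, seamless scale-pattern surface.
# The function takes the scale normal and mask map and outputs a rotated normal and a gradient mask.
# A Fresnel on the flattened rotated normal drives a sine-shaped thin-film spectrum lookup, limited by the gradient mask, as an emissive swipe.
# A reptilian gradient texture colors the base, and one texture channel feeds both roughness and metallic.
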Material /Game/Global/Materials/MaterialGym/Materials/RotateNormalInTangentSpace/M_RotateN_Example
graph {
 properties {bUseMaterialAttributes=true, pixelDepthOffsetMode=PDOM_Legacy, usedShadingModels=Default Lit}
 nodes [
  Constant() {r=1}
  n1 = NamedRerouteDeclaration(None=Multiply(A=TextureCoordinate(), B=Constant2Vector() {g=8, r=16}) {constB=8}) {name=UV, nodeColor=<rgba 0.4062 0 1 1>}
  n2 = TextureSample(UVs=@n1) {texture=/Game/Global/Materials/MaterialGym/Materials/RotateNormalInTangentSpace/Textures/T_ScalePattern_Seamless_M}
  NamedRerouteDeclaration(None=@n2.G) {name=ScaleMask, nodeColor=<rgba 0.125 1 0 1>}
  NamedRerouteDeclaration(None=@n2.R) {name=AOMask, nodeColor=<rgba 0.125 0 1 1>}
  n3 = TextureSample(UVs=@n1) {
   samplerType=SAMPLERTYPE_Normal
   texture=/Game/Global/Materials/MaterialGym/Materials/RotateNormalInTangentSpace/Textures/T_ScalePattern_Seamless_N
  }
  n4 = MaterialFunctionCall!(
   Material=SetMaterialAttributes!(MaterialAttributes, Base Color, Normal=@n3.RGB) {attributeSetTypes [#1, #2]}
   M_Map=@n2.RGB
   Tech Art Mask
  ) {materialFunction=&/MF_RotateN_TangentSpace}
  n5 = MaterialFunctionCall(Normal=@n4.Rotated Normal, Flatness=ScalarParameter() {defaultValue=0.8, parameterName=ExtraThinFilmNormalContribution}) {materialFunction=&FlattenNormal}
  n6 = Add(
   A=Fresnel(Normal=Transform(None=@n5.Result)) {baseReflectFraction=0, exponent=0.1}
   B=ScalarParameter() {defaultValue=-0.55, parameterName=ExtraThinFilmOffset}
  ) {constB=0.5}
  n7 = ScalarParameter() {defaultValue=1.4, parameterName=ExtraThinFilmScale}
  n8 = ConstantBiasScale(None=Sine(None=Multiply(A=@n6, B=@n7) {constB=1.5}))
  n9 = Multiply(
   A=TextureSample(UVs=Multiply(A=@n8)) {texture=/Game/Global/Textures/Default/Thinfilm/T_ThinFilm_Spectrum_COLOR}.RGB
   B=Saturate(None=Power(Base=@n4.GradientMask) {constExponent=4})
  )
  n10 = AppendVector(A=LinearInterpolate(Alpha=NamedRerouteUsage()) {constB=0.3}, B=Constant())
  n11 = TextureSample(UVs=NamedRerouteUsage()) {texture=/Game/Global/Materials/MaterialGym/Materials/RotateNormalInTangentSpace/Textures/T_ScalePattern_Seamless_S}
  SetMaterialAttributes!(
   MaterialAttributes
   Normal=@n4.Rotated Normal
   Emissive Color=NamedRerouteDeclaration(None=Multiply(A=@n9) {constB=0.25}) {name=ExtraThinFilmSwipe, nodeColor=<rgba 1 0 0.8906 1>}
   Base Color=Multiply(
    A=NamedRerouteUsage()
    B=TextureSample(UVs=@n10) {texture=/Game/Characters/Player/Female/Medium/Bodies/F_MED_ReptilianOcean_Sleek/Textures/ReptilianOceanSleek_Gradient}.RGB
   )
   Roughness=@n11.G
   Metallic=@n11.G
  ) {attributeSetTypes [
    #2
    #3
    #1
    #4
    #5
   ]}
 ]
}
